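# Dam surface material for Apollo that uses UV1.x to choose between concrete, walkway, sidewalk and wall texture sets, with a sidewalk-only static switch.
# Main concrete is tinted by actor-relative position and world normal.
# A UV2 environment texture adds dirt and roughness variation, then an optional switch blends in landscape RVT by red vertex color, bypassed for proxies.
# A world-aligned, height-blended cliff rock layer exists outside that final chain.
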Material /Game/Athena/Apollo/Environments/BuildingActors/Dam/Materials/M_Dam_Upd
graph {
 properties {
  bTangentSpaceNormal=false
  bUseMaterialAttributes=true
  bUsedWithInstancedStaticMeshes=true
  bUsedWithNanite=true
  pixelDepthOffsetMode=PDOM_Legacy
  usedShadingModels=Default Lit
 }
 nodes [
  n1 = ComponentMask(None=TextureCoordinate() {coordinateIndex=1}) {r=true}
  n2 = Constant() {r=1}
  n3 = VectorParameter() {group=Base colors, parameterName=Tint Front, defaultValue=<rgba 0.7396 0.6729 0.5701 0>}
  n4 = VectorParameter() {group=Base colors, parameterName=Front Sides, defaultValue=<rgba 0.4271 0.4182 0.4182 0>}
  n5 = Subtract(A=WorldPosition().XYZ, B=ActorPositionWS())
  n6 = ComponentMask(None=@n5) {r=true}
  n7 = MaterialFunctionCall(
   In=Divide(A=Abs(None=@n6), B=ScalarParameter() {defaultValue=4361.7446, group=Base colors, parameterName=Sides Tone Fade})
   Contrast=ScalarParameter() {defaultValue=1, group=Base colors, parameterName=Sides Contrast}
  ) {materialFunction=&CheapContrast}
  n8 = Saturate(None=ComponentMask(None=VertexNormalWS()) {g=true})
  n9 = TextureSample() {texture=/Game/Environments/Apollo/Sets/PowerPlant/Textures/T_PP_CoolingTower_Concrete1_D}
  n10 = ComponentMask(None=@n5) {b=true}
  n11 = Add(A=@n10, B=ScalarParameter() {defaultValue=4709.8267, parameterName=Bottom Offset})
  n12 = ScalarParameter() {defaultValue=3659.4739, parameterName=Bottom Ramp}
  n13 = LinearInterpolate(
   A=VectorParameter() {group=Base colors, parameterName=Tint Front Bottom, defaultValue=<rgba 0.7396 0.5729 0.3659 0>}.RGB
   Alpha=Saturate(None=Divide(A=@n11, B=@n12))
  ) {constA=1}
  n14 = Add(A=@n10, B=ScalarParameter() {defaultValue=4709.8267, parameterName=Bottom Offset2})
  n15 = ScalarParameter() {defaultValue=3098.6094, parameterName=Bottom Ramp2}
  n16 = LinearInterpolate(
   A=VectorParameter() {group=Base colors, parameterName=Tint Front Bottom2, defaultValue=<rgba 0.313 0.3385 0.0846 0>}.RGB
   Alpha=Saturate(None=Divide(A=@n14, B=@n15))
  ) {constA=1}
  n17 = Multiply(
   A=LinearInterpolate(
    A=LinearInterpolate(A=@n3.RGB, B=@n4.RGB, Alpha=@n7.Result)
    B=VectorParameter() {group=Base colors, parameterName=Tint walls, defaultValue=<rgba 0.5104 0.3931 0.3137 0>}.RGB
    Alpha=@n8
   )
   B=Multiply(A=@n9.RGB, B=Multiply(A=@n13, B=@n16))
  )
  n18 = TextureSample() {texture=/Game/Environments/Apollo/Sets/PowerPlant/Textures/T_PP_Walkway_D}
  n19 = TextureSample() {texture=/Game/Environments/Apollo/Sets/PowerPlant/Textures/T_PP_Sidewalk_D}
  n20 = If(
   A=Constant() {r=3}
   B=@n1
   "A > B"=If(A=Constant() {r=2}, B=@n1, "A > B"=If(A=@n2, B=@n1, "A > B"=@n17, "A < B"=@n18.RGB) {desc=0}, "A < B"=@n19.RGB)
   "A < B"=TextureSample() {texture=/Game/Environments/Apollo/Sets/PowerPlant/Textures/T_PP_Concrete_Wall_2_D}.RGB
  )
  n21 = ComponentMask(None=TextureCoordinate() {coordinateIndex=1}) {r=true}
  n22 = Constant() {r=2}
  n23 = ScalarParameter() {defaultValue=0.2, group=Base colors, parameterName=S Wall}
  n24 = MaterialFunctionCall(Float3=TextureSample() {samplerType=SAMPLERTYPE_Masks, texture=/Game/Environments/Apollo/Sets/PowerPlant/Textures/T_PP_CoolingTower_Concrete1_S}.RGB) {materialFunction=&BreakOutFloat3Components}
  n25 = LinearInterpolate(A=LinearInterpolate(A=ScalarParameter() {defaultValue=0.5, group=Base colors, parameterName=R Mid}, Alpha=@n7.Result) {constA=1}, Alpha=@n8)
  n26 = Multiply(A=@n24.B, B=ScalarParameter() {defaultValue=1.5, group=Base colors, parameterName=R Wall})
  n27 = If(
   A=Constant() {r=1}
   B=@n21
   "A > B"=MaterialFunctionCall(X=Multiply(A=@n9.R, B=@n23), Y=@n24.G, Z=Multiply(A=@n25, B=@n26)) {materialFunction=&MakeFloat3}.Result
   "A < B"=TextureSample() {samplerType=SAMPLERTYPE_Masks, texture=/Game/Environments/Apollo/Sets/PowerPlant/Textures/T_PP_Walkway_S}.RGB
  ) {desc=0}
  n28 = TextureSample() {samplerType=SAMPLERTYPE_Masks, texture=/Game/Environments/Apollo/Sets/PowerPlant/Textures/T_PP_Sidewalk_S}
  n29 = MaterialFunctionCall(Float3=If(
    A=Constant() {r=3}
    B=@n21
    "A > B"=If(A=@n22, B=@n21, "A > B"=@n27, "A < B"=@n28.RGB)
    "A < B"=TextureSample() {samplerType=SAMPLERTYPE_Masks, texture=/Game/Environments/Apollo/Sets/PowerPlant/Textures/T_PP_Concrete_Wall_2_S}.RGB
   )) {materialFunction=&BreakOutFloat3Components}
  n30 = Constant() {r=3}
  n31 = ComponentMask(None=TextureCoordinate() {coordinateIndex=1}) {r=true}
  n32 = TextureSample() {samplerType=SAMPLERTYPE_Normal, texture=/Game/Environments/Apollo/Sets/PowerPlant/Textures/T_PP_Sidewalk_N}
  n33 = If(
   A=Constant() {r=2}
   B=@n31
   "A > B"=If(
    A=Constant() {r=1}
    B=@n31
    "A > B"=TextureSample() {samplerType=SAMPLERTYPE_Normal, texture=/Game/Environments/Apollo/Sets/PowerPlant/Textures/T_PP_CoolingTower_Concrete1_N}.RGB
    "A < B"=TextureSample() {samplerType=SAMPLERTYPE_Normal, texture=/Game/Environments/Apollo/Sets/PowerPlant/Textures/T_PP_Walkway_N}.RGB
   ) {desc=0}
   "A < B"=@n32.RGB
  )
  n34 = TextureSample() {samplerType=SAMPLERTYPE_Normal, texture=/Game/Environments/Apollo/Sets/PowerPlant/Textures/T_PP_Concrete_Wall_2_N}
  n35 = MakeMaterialAttributes!(
   BaseColor=@n20
   Metallic
   Specular=@n29.R
   Roughness=@n29.R
   Anisotropy
   EmissiveColor
   Opacity
   OpacityMask
   Normal=Transform(None=If(A=@n30, B=@n31, "A > B"=@n33, "A < B"=@n34.RGB))
   Tangent
   WorldPositionOffset
   SubsurfaceColor
   ClearCoat
   ClearCoatRoughness
   AmbientOcclusion
   Refraction
   CustomizedUVs_0
   CustomizedUVs_1
   CustomizedUVs_2
   CustomizedUVs_3
   CustomizedUVs_4
   CustomizedUVs_5
   CustomizedUVs_6
   CustomizedUVs_7
   PixelDepthOffset
   ShadingModel
   Displacement
  )
  n36 = MaterialFunctionCall!(
   Diffuse Texture Object=TextureObject() {texture=/Game/Environments/Apollo/Sets/Terrain/Textures/T_Apollo_Cliff_D_01}
   Spec Texture Object=TextureObject() {samplerType=SAMPLERTYPE_Masks, texture=/Game/Environments/Apollo/Sets/Terrain/Textures/T_Apollo_Cliff_S_01}
   TextureSize=Multiply(A=Constant3Vector() {constant=<rgba 2555 2555 2000 0>}, B=ScalarParameter() {defaultValue=0.7, group=Rocks, parameterName=ScaleRocks})
   --------------------------
   Use seperate settings for Z projection
   Use seperate texture for Z projection
   Texture Size Z projection
   Z Plane Diffuse Texture Object
   Z Plane Spec Texture Object
   "--------------------------      "
   ProjectionTransitionContrast
   Use Texture Alpha Heightmap to Refine the Transistion
   HeightLerpTransitionContrast
   PickTheDiffuseHeightChannel
   TheDiffuseHeightChannel
   "--------------------------       "
   NormalTextureObject=TextureObject() {samplerType=SAMPLERTYPE_Normal, texture=/Game/Environments/Apollo/Sets/Terrain/Textures/T_Apollo_Cliff_N_01}
   World Space Normals
   "--------------------------  "
   Export Float 4 Diffuse=StaticBool() {value=true}
   Export Float 4 Spec
   "--------------------------     "
   WorldPosition
   World Space Normal
  ) {materialFunction=/Engine/Functions/Engine_MaterialFunctions02/Texturing/WorldAlignedTextures_Complex}
  n37 = ComponentMask(None=@n36.XY Diffuse Texture) {b=true, g=true, r=true}
  n38 = LinearInterpolate(A=MaterialFunctionCall!(In) {materialFunction=/Game/Environments/Apollo/VS/Materials/MF_DetailTexture}.Texture Only) {constB=0.5}
  n39 = MaterialFunctionCall(Float3=@n36.XY Spec) {materialFunction=&BreakOutFloat3Components}
  n40 = Constant() {r=0.25}
  n41 = ScalarParameter() {defaultValue=0.5, group=Rocks, parameterName=Rocks Flatten N}
  n42 = Subtract(A=WorldPosition().XYZ, B=ActorPositionWS())
  n43 = Divide(A=Abs(None=ComponentMask(None=@n42) {r=true}), B=ScalarParameter() {defaultValue=7000, group=Rocks, parameterName=Sides Fade Rocks})
  n44 = ScalarParameter() {defaultValue=5, group=Rocks, parameterName=Sides Tone Fade Rocks Contrast}
  n45 = VertexNormalWS()
  n46 = LinearInterpolate(
   A=Saturate(None=MaterialFunctionCall(In=@n43, Contrast=@n44) {materialFunction=&CheapContrast}.Result)
   Alpha=Saturate(None=ComponentMask(None=@n45) {g=true})
  ) {constB=0}
  n47 = ComponentMask(None=@n36.XY Diffuse Texture) {a=true}
  n48 = ScalarParameter() {defaultValue=3, group=Rocks, parameterName=Rocks Contrast}
  BlendMaterialAttributes(
   A=@n35
   B=MakeMaterialAttributes(
    BaseColor=MaterialFunctionCall(Base=@n37, Blend=@n38) {materialFunction=&Blend_Overlay}.Result
    Metallic=@n39.G
    Specular=Multiply(A=@n39.R, B=@n40)
    Roughness=@n39.G
    Normal=MaterialFunctionCall(Normal=@n36.Normal XY Texture, Flatness=@n41) {materialFunction=&FlattenNormal}.Result
   )
   Alpha=Multiply(
    A=VertexColor().G
    B=MaterialFunctionCall(Transition Phase=@n46, Height Texture=@n47, Contrast=@n48) {materialFunction=&HeightLerp}.Alpha
   )
  )
  MakeMaterialAttributes(BaseColor=@n20, Normal=VertexNormalWS())
  n49 = GetMaterialAttributes(None=StaticSwitchParameter(True=MakeMaterialAttributes(BaseColor=@n19.RGB, Metallic=@n28.G, Specular=@n28.G, Roughness=@n28.G, Normal=@n32.RGB), False=@n35) {parameterName=SidewalkOnly}) {attributeGetTypes [
    #1
    #2
    #3
    #4
   ]}
  n50 = TextureSample(UVs=TextureCoordinate() {coordinateIndex=2}) {texture=/Game/Athena/Apollo/Environments/BuildingActors/Dam/Textures/T_Environment_D}
  n51 = ScalarParameter() {defaultValue=-0.7, group=Decal Dirt, parameterName=R Wet Min}
  n52 = ScalarParameter() {group=Decal Dirt, parameterName=R Wet Max}
  n53 = LinearInterpolate(
   A=ScalarParameter() {defaultValue=-0.4952, group=Decal Dirt, parameterName=R Mask Min}
   B=ScalarParameter() {defaultValue=2, group=Decal Dirt, parameterName=R Mask Max}
   Alpha=@n50.R
  )
  n54 = SetMaterialAttributes!(
   MaterialAttributes=@n49.MaterialAttributes
   Base Color=Multiply(A=@n50.RGB, B=@n49.Base Color)
   Metallic=@n49.MaterialAttributes
   Specular=Multiply(A=@n49.Specular, B=@n50.R)
   Roughness=LinearInterpolate(A=LinearInterpolate(A=@n51, B=@n52, Alpha=@n49.Roughness), B=@n49.Roughness, Alpha=Saturate(None=@n53))
  ) {attributeSetTypes [
    #1
    #2
    #3
    #4
   ]}
  n55 = RuntimeVirtualTextureSample!(UVs, World Position) {materialType=BaseColor_Normal_Specular_YCoCg, virtualTexture=/Game/Athena/Apollo/Maps/Landscape/VT/RVT_Apollo_Landscape}
  n56 = SetMaterialAttributes!(MaterialAttributes, Base Color=@n55.BaseColor, Specular=@n55.BaseColor, Roughness=@n55.BaseColor, Normal=@n55.BaseColor) {attributeSetTypes [
    #1
    #3
    #4
    #5
   ]}
  n57 = BlendMaterialAttributes(A=@n56, B=@n54, Alpha=OneMinus(None=VertexColor().R))
  MaterialProxyReplace(
   Realtime=StaticSwitchParameter(
    True=MaterialFunctionCall(NoRVT=@n54, RVT=@n57) {materialFunction=&FN_RVT_Switch_Material}.Result
    False=@n54
   ) {group=RVT, parameterName=RVT Enable}
   MaterialProxy=@n54
  )
 ]
}
